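; Blueprint actor for a low-gravity zone around a glitching meteor.
; Overlapping ZeroRadius applies low-gravity gameplay effects to player pawns or vehicles; leaving removes them by tag, and players ignore their next falling damage.
; Non-server instances disable that collision and run a randomly timed, self-rescheduling glitch loop.
; Each glitch activates a wave particle, plays ambient audio and drives the shared MPC_Nexus Glitch parameter through a randomly paced timeline.
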(event EventBeginPlay
  (if (not (Networking|IsServer))
    (Collision|SetCollisionProfileName (Variables|Default|GetZeroRadius) "nocollision" true :!class /Script/Engine.PrimitiveComponent)
    (Utilities|Time|SetTimerbyEvent (EventDispatchers|CreateEvent :!event Custom|Glitch) (Math|Random|RandomFloatinRange 2.5 6.0))))

(event Custom|Glitch
  (Components|Activation|Activate (Variables|Default|GetP_Meteor_Glitch_Wave) true)
  (Components|Timeline|SetPlayRate (Variables|BP_Space_Gravy|GetTimeline_0) (Math|Random|RandomFloatinRange 0.8 3.0))
  (Audio|PlaySoundatLocation "/Figment_S10/Athena/Sounds/Events/Meteor/Meteor_Ambient_Glitch_Cue.Meteor_Ambient_Glitch_Cue" (Transformation|GetWorldLocation (Variables|Default|GetAmbientAudio)))
  (bind (_direction _glitch) (Timeline|Timeline_0 PlayFromStart)
    (:Update
      (Rendering|Material|SetScalarParameterValue "/Game/Athena/Environments/Nexus/Materials/FXshader/MPC_Nexus.MPC_Nexus" "Glitch" _glitch :!class /Script/Engine.KismetMaterialLibrary))
    (:Finished
      (Utilities|Time|SetTimerbyEvent (EventDispatchers|CreateEvent :!event Custom|Glitch) (Math|Random|RandomFloatinRange 2.5 6.0)))))

(event OnComponentBeginOverlap(ZeroRadius)
  (params (OverlappedComponent :type Primitive%20Component%20Object%20Reference) (OtherActor :type Actor%20Object%20Reference) (OtherComp :type Primitive%20Component%20Object%20Reference) (OtherBodyIndex :type Integer) (bFromSweep :type Boolean) (SweepResult :type Hit%20Result%20Structure%20%28by%20ref%29))
  (bind _asplayer_pawn_athena (Utilities|Casting|CastToPlayerPawn_Athena OtherActor)
    (:then
      (Environment|ApplyGlobalEnvironmentGameplayEffecttoActor _asplayer_pawn_athena "/Figment_S10/Athena/Environments/Cattus/GE_LowGravZone_Meteor.GE_LowGravZone_Meteor_C"))
    (:CastFailed
      (bind _asfort_athena_vehicle (Utilities|Casting|CastToFortAthenaVehicle OtherActor))
      (Environment|ApplyGlobalEnvironmentGameplayEffecttoActor _asfort_athena_vehicle "/Figment_S10/Athena/Environments/Cattus/GE_LowGravZone_Vehicle.GE_LowGravZone_Vehicle_C"))))

(event OnComponentEndOverlap(ZeroRadius)
  (params (OverlappedComponent :type Primitive%20Component%20Object%20Reference) (OtherActor :type Actor%20Object%20Reference) (OtherComp :type Primitive%20Component%20Object%20Reference) (OtherBodyIndex :type Integer))
  (bind _asplayer_pawn_athena (Utilities|Casting|CastToPlayerPawn_Athena OtherActor)
    (:then
      (Variables|/Script/FortniteGame.FortPawn|SetbIgnoreNextFallingDamage true _asplayer_pawn_athena)
      (GameplayEffects|RemoveActiveEffectswithGrantedTags (Ability|GetAbilitySystemComponent _asplayer_pawn_athena) "(GameplayTags=((TagName=\"Status.LowGravity.LowGravZone\")))"))
    (:CastFailed
      (bind _asfort_athena_vehicle (Utilities|Casting|CastToFortAthenaVehicle OtherActor))
      (GameplayEffects|RemoveActiveEffectswithGrantedTags (Ability|GetAbilitySystemComponent _asfort_athena_vehicle) "(GameplayTags=((TagName=\"Status.LowGravity.LowGravZone\")))"))))

(fn ConstructionScript)
