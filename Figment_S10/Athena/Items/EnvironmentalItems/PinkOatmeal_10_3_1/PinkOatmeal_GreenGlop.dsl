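; Green Glop rift-bubble volume (PinkOatmeal) that runs timed "Taco Time" pulses with dancing, music and foraged pickups.
; Server sets PulseState, times the pulse and spawns glop pickups around the bubble at random intervals.
; Dance GameplayEffect follows pulse and movement mode; OnRep drives music, confetti and FFT-reactive Material Parameter Collection colors, with low-pass filtering by player position.
; BeginPlay proximity binds are severed (S10 extraction), leaving pulse handlers unbound here.
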(event Custom|SpawnCBGA
  (params (Spawn%20Transform%20Location :type Vector))
  ;[1 1 1] Spawn Th pickup
  (if (Networking|IsServer)
    (Game|SpawnActorfromClass "/Figment_S10/Athena/Items/ForagedItems/GreenGlop/CBGA_GreenGlop_WithGrav.CBGA_GreenGlop_WithGrav_C" (Math|Transform|MakeTransform Spawn%20Transform%20Location (Math|Rotator|MakeRotator 0.0 0.0 (Math|Random|RandomFloatinRange 0.0 3.6e+02))) "AlwaysSpawn" "OverrideRootScale" 0 0 (Game|GetInstigator))))

(event Custom|TacoStopTime
  ;[1 1 1] On interval Spawn Pickup during Taco Time
  (Variables|Default|SetPulseState 0))

(event Custom|TacoStart
  ;[1 1 1] On interval Spawn Pickup during Taco Time
  (if (Networking|IsServer)
    (Variables|Default|SetPulseState 2)
    (Utilities|Time|SetTimerbyEvent (delegate Custom|TacoStopTime) (ScalableFloat|GetValueatLevel (Variables|Default|GetRow_PulseDuration)))
    (if (Math|Conversions|ToBoolean(Integer) (Math|Float|Ceil (ScalableFloat|GetValueatLevel (Variables|Default|GetRow_ForagablesEnabled))))
      (CallFunction|SpawnCBGA (CallFunction|MakeSpawnLocation))
      (Utilities|FlowControl|Delay (Math|Random|RandomFloatinRange (ScalableFloat|GetValueatLevel (Variables|Default|GetRow_SpawnIntervalMin)) (ScalableFloat|GetValueatLevel (Variables|Default|GetRow_SpawnIntervalMax))))
      (if (== (Variables|Default|GetPulseState) 2)
        _))))

(event EventBeginPlay
  ;[1 1 1] EVENT -> BeginPlay
  ;; S10 Extraction - Severed this connection since the Binds needed a Target
  (|CallParentFunction))

(event EventOnActorEnterPinkOatmeal
  (params (Actor :type Actor%20Object%20Reference) (ActorComponent :type Actor%20Component%20Object%20Reference) (InstigatorPawn :type Pawn%20Object%20Reference) (Volume :type Fort%20Volume%20Object%20Reference))
  ;[1 1 1] EVENT -> OnActorEnterPinkOatmeal, if the pulse is active when we enter, then apply the pulse GE.
  (|CallParentFunction Actor ActorComponent InstigatorPawn Volume)
  (bind _asfort_player_pawn (Utilities|Casting|CastToFortPlayerPawn InstigatorPawn))
  (Character|BindEventtoMovementModeChangedDelegate _asfort_player_pawn (EventDispatchers|CreateEvent :!event Custom|MovementModeChanged_Bind)))

(event EventOnActorExitPinkOatmeal
  (params (Actor :type Actor%20Object%20Reference) (ActorComponent :type Actor%20Component%20Object%20Reference) (InstigatorPawn :type Pawn%20Object%20Reference) (Volume :type Fort%20Volume%20Object%20Reference))
  ;[1 1 1] EVENT -> OnActorExitPinkOatmeal, removing the GE that we apply on pulse.
  (|CallParentFunction Actor ActorComponent InstigatorPawn Volume)
  (CallFunction|RemoveDance InstigatorPawn)
  (bind _asfort_player_pawn (Utilities|Casting|CastToFortPlayerPawn InstigatorPawn))
  (Character|UnbindEventfromMovementModeChangedDelegate _asfort_player_pawn (EventDispatchers|CreateEvent :!event Custom|MovementModeChanged_Bind)))

(event Custom|MovementModeChanged_Bind
  (params (Character :type Character%20Object%20Reference) (PrevMovementMode :type EMovementMode%20Enum :default "MOVE_None") (PreviousCustomMode :type Byte))
  ;[1 1 1] CUSTOM EVENT -> MovementModeChanged_Bind
  (bind _asfort_player_pawn (Utilities|Casting|CastToFortPlayerPawn Character))
  (if (and (Utilities|Enum|Equal(Enum) (Variables|/Script/Engine.CharacterMovementComponent|GetMovementMode (Variables|/Script/Engine.Character|GetCharacterMovement _asfort_player_pawn)) "MOVE_Walking") (Variables|Default|GetWaitingToWalkAgain))
    (CallFunction|AddDance _asfort_player_pawn)
    (Variables|Default|SetWaitingToWalkAgain false)
    (elif (Math|Boolean|ANDBoolean (not (Variables|Default|GetWaitingToWalkAgain)) (Utilities|Enum|NotEqual(Enum) (Variables|/Script/Engine.CharacterMovementComponent|GetMovementMode (Variables|/Script/Engine.Character|GetCharacterMovement _asfort_player_pawn)) "MOVE_Walking") (Utilities|Enum|NotEqual(Enum) (Variables|/Script/Engine.CharacterMovementComponent|GetMovementMode (Variables|/Script/Engine.Character|GetCharacterMovement _asfort_player_pawn)) "MOVE_Falling"))
      (CallFunction|RemoveDance _asfort_player_pawn)
      (Variables|Default|SetWaitingToWalkAgain true))))

(event Custom|OnProximityPrePulse_Bind
  ;[1 1 1] CUSTOM EVENT -> OnProximityPrePulse_Bind
  (Variables|Default|SetPulseState 1))

(event Custom|OnProximityPulse_Bind
  ;[1 1 1] CUSTOM EVENT -> OnProximityPulse_Bind
  (CallFunction|TacoStart))

(event Custom|OnProximityEffectsApplied_Bind
  (params (TargetActor :type Actor%20Object%20Reference) (BucketTag :type Gameplay%20Tag%20Structure))
  ;[1 1 1] CUSTOM EVENT -> OnProximityEffectsApplied_Bind
  (if (GameplayTags|Equal(GameplayTag) BucketTag "(TagName=\"\")")
    (CallFunction|AddDance TargetActor)))

(event EventOnPlayerInsideChanged
  (params (NewInside :type Boolean))
  ;[1 1 1] EVENT -> OnPlayerInsideChanged
  (if NewInside
    (bind (_direction _lpf) (Timeline|Music%20LPF Play)
      (:Update
        (Audio|Components|Audio|SetLowPassFilterFrequency (Variables|Default|GetMusicAudio) (Math|Float|MapRangeClamped _lpf 0.0 1.0 5e+02 2e+04) :!class /Script/Engine.AudioComponent)))
    (else
      (Timeline|Music%20LPF Reverse))))

(event Custom|Update%20FFT%20Data
  ;[1 1 1] CUSTOM EVENT -> UpdateFFTData
  (bind (_outsoundwavespectraldata _returnvalue) (Audio|Components|Audio|GetCookedFFTDataforAllPlayingSounds (Variables|Default|GetMusicAudio)))
  (if _returnvalue
    (Utilities|FlowControl|Sequence
      (:then_0
        ;;[1 1 1] To Make Rift Bubble Colors Dance to the Music
        (bind (_output_frequencyhz _output_magnitude _output_normalizedmagnitude) (Utilities|Struct|BreakSoundWaveSpectralData (Utilities|Array|Get(aref) (Utilities|Struct|BreakSoundWaveSpectralDataPerSound (Utilities|Array|Get(aref) _outsoundwavespectraldata)))))
        (Rendering|Material|SetScalarParameterValue "/Game/Environments/Props/RiftBeacon/FX/MPC_RiftBubble_GreenGoop.MPC_RiftBubble_GreenGoop" "TacoTime_AudioLow" _output_normalizedmagnitude :!class /Script/Engine.KismetMaterialLibrary)))))

(fn AddDance
  (params (Object :type Object%20Reference))
  (bind _asfort_player_pawn (Utilities|Casting|CastToFortPlayerPawn Object))
  (Utilities|IsPulseActive
    (:True
      (Fortnite|ApplyGameplayEffectsToActor _asfort_player_pawn (Utilities|Array|MakeArray (Variables|Default|GetGE_PinkOatmeal_Dance)) (Transformation|GetActorLocation _asfort_player_pawn) false 0 (Ability|GetAbilitySystemComponent self)))))

(fn MakeSpawnLocation
  (bind (_returnvalue_x _returnvalue_y) (Math|Vector|BreakVector (* (Math|Rotator|GetRotationXVector (Math|Rotator|MakeRotator 0.0 0.0 (Math|Random|RandomFloatinRange 0.0 3.6e+02))) (Math|Random|RandomFloatinRange (LocalVariables|GetSpawnRadiusMin) (ScalableFloat|GetValueatLevel (Variables|Default|GetRow_SpawnRadius))))))
  (bind (_returnvalue_x_2 _returnvalue_y_2 _returnvalue_z) (Math|Vector|BreakVector (Transformation|GetActorLocation)))
  (return (Math|Vector|MakeVector (+ _returnvalue_x_2 _returnvalue_x) (+ _returnvalue_y_2 _returnvalue_y) (+ _returnvalue_z (LocalVariables|GetSpawnHeight)))))

(fn OnRep_PulseActive
  (bind _outputdelegate (EventDispatchers|CreateEvent :!event Update%20FFT%20Data))
  (bind _playerinside (Variables|Audio|GetPlayerInside))
  (bind _returnvalue (select _playerinside 2e+04 5e+02))
  (bind _pulsestate (Variables|Default|GetPulseState))
  (bind _worldconfetti (Variables|Default|GetWorldConfetti))
  (bind _premusicaudio (Variables|Default|GetPreMusicAudio))
  (bind _particlesystemcomponent (Variables|/Script/CascadeRuntime.Emitter|GetParticleSystemComponent _worldconfetti))
  (bind _musicaudio (Variables|Default|GetMusicAudio))
  (if (== _pulsestate 2)
    (Audio|Components|Audio|Play _musicaudio 0.0)
    (Utilities|Time|SetTimerbyEvent _outputdelegate 0.1 true)
    (Audio|Components|Audio|SetLowPassFilterFrequency _musicaudio _returnvalue :!class /Script/Engine.AudioComponent)
    (Utilities|IsValid _worldconfetti
      (:Is%20Valid
        (Components|Activation|Activate _particlesystemcomponent false)
        (Rendering|Material|SetScalarParameterValueonMaterials (Variables|Default|GetAurora) "Alpha" 0.0)
        (Rendering|Material|SetScalarParameterValue "/Game/Environments/Props/RiftBeacon/FX/MPC_RiftBubble_GreenGoop.MPC_RiftBubble_GreenGoop" "TACOTIME" 1.0 :!class /Script/Engine.KismetMaterialLibrary)))
    (elif (== _pulsestate 1)
      (Audio|Components|Audio|FadeIn _premusicaudio 1.0 1.0 0.0 "Linear" :!class /Script/Engine.AudioComponent)
      (Utilities|Time|SetTimerbyEvent _outputdelegate 0.1 true)
      (Audio|Components|Audio|SetLowPassFilterFrequency _premusicaudio _returnvalue :!class /Script/Engine.AudioComponent)
      (elif (== _pulsestate 0)
        (Utilities|FlowControl|Sequence
          (:then_0
            (Audio|Components|Audio|FadeOut (Variables|Default|GetMusicAudio) 0.5 0.0 "Linear" :!class /Script/Engine.AudioComponent)
            (Audio|PlaySoundatLocation "/Figment_S10/Athena/Sounds/Events/RiftBeacon/RiftBubble/GreenGlop/RiftBubble_GreenGlop_End_Music_Global_Cue.RiftBubble_GreenGlop_End_Music_Global_Cue" (Transformation|GetWorldLocation (Variables|Default|GetMusicAudio)))
            (Utilities|Time|SetTimerbyEvent (EventDispatchers|CreateEvent :!event Update%20FFT%20Data))
            (Utilities|IsValid _worldconfetti
              (:Is%20Valid
                (Components|Activation|Deactivate _particlesystemcomponent)
                (Rendering|Material|SetScalarParameterValueonMaterials (Variables|Default|GetAurora) "Alpha" 1.0)
                (Rendering|Material|SetScalarParameterValue "/Game/Environments/Props/RiftBeacon/FX/MPC_RiftBubble_GreenGoop.MPC_RiftBubble_GreenGoop" "TACOTIME" :!class /Script/Engine.KismetMaterialLibrary))))
          (:then_1
            ;[1 1 1] Removing the dance GE on pulse end in case we added it when the player entered while the pulse was already active.
            (for _array_element (Volume|GetActorsWithinVolumebyClass "/Script/FortniteGame.FortPlayerPawn")
              (bind _asfort_player_pawn (Utilities|Casting|CastToFortPlayerPawn _array_element))
              (GameplayEffects|RemoveActiveGameplayEffectbySourceEffect (Ability|GetAbilitySystemComponent _asfort_player_pawn) (Variables|Default|GetGE_PinkOatmeal_Dance) (Ability|GetAbilitySystemComponent self) -1))))))))

(fn RemoveDance
  (params (Object :type Object%20Reference))
  (bind _asfort_player_pawn (Utilities|Casting|CastToFortPlayerPawn Object))
  (Utilities|IsPulseActive
    (:True
      (GameplayEffects|RemoveActiveGameplayEffectbySourceEffect (Ability|GetAbilitySystemComponent _asfort_player_pawn) (Variables|Default|GetGE_PinkOatmeal_Dance) (Ability|GetAbilitySystemComponent self) -1))))

(fn ConstructionScript)
